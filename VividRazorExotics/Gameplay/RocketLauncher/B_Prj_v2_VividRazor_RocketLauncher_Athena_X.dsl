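; Rocket projectile for the VividRazor launcher that explodes with environmental knockback, granting fall damage immunity and firing launch cues on actors in range.
; OnStop or a fuse timer triggers the blast at an offset from the decal location; delayed water checks can detonate it early.
; Hits on "ShouldCauseBounce" components double velocity instead of stopping.
; Exploded and DisableForKill duplicate cosmetic logic: water-aware FX, sound, radial force, post-process fade and feedback cue.
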(event EventBeginPlay
  ;[1 1 1] EVENT -> Begin Play
  (Utilities|FlowControl|Sequence
    (:then_0
      ;;[1 1 1] Timer to enable touching water to cause an explosion
      (if (> (Variables|Effects|GetWaterInteractionDelay) 0.0)
        (Utilities|Time|SetTimerbyEvent (EventDispatchers|CreateEvent :!event Custom|RocketWaterDelay) (Variables|Effects|GetWaterInteractionDelay))
        (else
          (Variables|Projectile|SetTouchWaterBehavior "StopIfStopSimulatingOnHit"))))
    (:then_1
      ;;[1 1 1] Spawn attached Rocket Trail
      (Utilities|Time|SetTimerbyFunctionName self "FuseTimerMax" 15)
      (Networking|CanExecuteCosmeticEvents
        (:True
          (bind _returnvalue (Effects|SpawnFXSystemAttached :FXSystemAsset (Variables|RocketTrail|GetRocket%20Trial%20Asset) :AttachToComponent (Variables|Collision|GetCapsuleComponent) :AttachPointName "None" :Rotation (Math|Rotator|MakeRotator (Math|Random|RandomFloatinRange 0.0 360)) :Scale "1.000000,1.000000,1.000000" :LocationType "KeepRelativeOffset" :bAutoDestroy true :PoolingMethod "None" :bAutoActivate true))
          (bind _asniagara_particle_system_component (Utilities|Casting|CastToNiagaraComponent _returnvalue)
            (:then
              (Variables|Default|SetIsNiagaraEnabled true)
              (bind _output_get (Variables|Default|SetRocketTrailPS_Niagara _asniagara_particle_system_component)))
            (:CastFailed
              (bind _ascascade_particle_system_component (Utilities|Casting|CastToParticleSystemComponent _returnvalue))
              (bind _output_get_2 (Variables|RocketTrail|SetRocketTrailPS _ascascade_particle_system_component))))
          (Rendering|SetRenderCustomDepth _output_get true)
          (Audio|Components|Audio|SetSound (Variables|Default|GetRocket_AudioComponent) (Variables|Sounds|GetRocket_Projectile_InAir_Sound)) ; Set Audio Component sound
        )))
    (:then_2
      ;;[1 1 1] Disable collision with self player pawn
      (Collision|IgnoreActorwhenMoving (Variables|Default|GetCollisionMesh) (Game|GetPlayerPawn 0) false))))

(event EventTick
  (params (DeltaSeconds :type Float%20%28single-precision%29))
  ;[1 1 1] EVENT -> Tick: spin the rocket and particle trail
  (Variables|Effects|SetTick%20Delta DeltaSeconds))

(event EventOnTouched
  (params (OtherActor :type Actor%20Object%20Reference) (OtherComp :type Primitive%20Component%20Object%20Reference) (HitResult :type Hit%20Result%20Structure%20%28by%20ref%29) (bIsOverlap :type Boolean))
  (bind (_bblockinghit _binitialoverlap _time _distance _location _impactpoint _normal _impactnormal _physmat _hitactor _hitcomponent) (Collision|BreakHitResult HitResult))
  ;[1 1 1] EVENT -> Touched: Bounce off "NoBounce"
  (if (Components|ComponentHasTag _hitcomponent "ShouldCauseBounce")
    (Variables|/Script/Engine.MovementComponent|SetVelocity (* (Variables|/Script/Engine.MovementComponent|GetVelocity (Variables|Movement|GetProjectileMovementComponent)) 2.0) (Variables|Movement|GetProjectileMovementComponent))
    (else
      (Variables|Projectile|SetbStopSimulatingOnHit true))))

(event Custom|RocketWaterDelay
  ;[1 1 1] Delay for rocket to explode in Water
  (if (and (Utilities|IsValid (Variables|/Script/Engine.MovementComponent|GetUpdatedComponent (Variables|Movement|GetProjectileMovementComponent)) :!class /Script/Engine.KismetSystemLibrary) (not (Delivery|IsBeingKilled)))
    (Variables|Projectile|SetTouchWaterBehavior "StopIfStopSimulatingOnHit")
    (if (Water|IsinWater (Variables|Water|GetWaterInteractionComponent))
      (CallFunction|FuseTimerMax))))

(event EventOnStop
  (params (Hit :type Hit%20Result%20Structure%20%28by%20ref%29))
  (bind (_bblockinghit _binitialoverlap _time _distance _location _impactpoint _normal _impactnormal _physmat _hitactor) (Collision|BreakHitResult Hit))
  (bind _yknockbackoffset (Variables|Default|GetYKnockbackOffset))
  (bind _returnvalue (ScalableFloat|GetValueatLevel _yknockbackoffset))
  (bind _zknockbackoffset (Variables|Default|GetZKnockbackOffset))
  (bind _returnvalue_2 (ScalableFloat|GetValueatLevel _zknockbackoffset))
  (bind _explosionradius (Variables|Default|GetExplosionRadius))
  (bind _returnvalue_3 (ScalableFloat|GetValueatLevel _explosionradius))
  (bind _collisionmesh (Variables|Default|GetCollisionMesh))
  (bind _returnvalue_4 (Transformation|GetWorldTransform _collisionmesh))
  (bind _returnvalue_5 (Math|Transform|TransformDirection _returnvalue_4 (Math|Vector|MakeVector -5.0 _returnvalue _returnvalue_2)))
  (bind _decallocation (Variables|Effects|Explosion|Decal|GetDecalLocation))
  (bind _returnvalue_6 (+ _decallocation _returnvalue_5))
  ;[1 1 1] EVENT -> Stop: Impact and Explosion Stuff
  (Variables|Default|SetHitNormal _normal)
  (Variables|Effects|Explosion|Decal|SetDecalLocation _impactpoint)
  (Variables|Effects|SetStopLocation (+ _impactpoint (* (Variables|Effects|Explosion|GetExplosion%20Impact%20Offset) _impactnormal)))
  (Variables|Effects|SetStopRotZ (Math|Rotator|MakeRotfromZ _impactnormal))
  (Utilities|IsValid _physmat
    (:Is%20Valid
      (Variables|Effects|SetSurfaceType (Variables|/Script/PhysicsCore.PhysicalMaterial|GetSurfaceType _physmat)))
    (:Is%20Not%20Valid
      (Variables|Effects|SetSurfaceType "SurfaceType_Default")))
  (Class|BPSurfaceTypeSoundComponent|SetSurfaceType (Variables|Sound|GetBP_SurfaceTypeSoundComponent) (Variables|Effects|GetSurfaceType) _hitactor)
  (if (not (Delivery|IsBeingKilled))
    (Delivery|DoExplosion)
    (Fortnite|ShowSoundIndicatorLocal :Location (Transformation|GetWorldLocation (Variables|Projectile_Rocket|GetMesh)) :MaxAudibleDistance 5e+03 :IndicatorType "Gunshot" :Instigator self :Tint "(R=0.000000,G=0.000000,B=0.000000,A=0.000000)" :AttenuationMaxRangeMultiplier 1.0 :AttenuationIntensityMultiplier 1.0)
    ;[0.003247 0.6351 1 0.5] Knockback
    (Class|AthenaFunctionLibrary|CreateEnvironmentalKnockbackExplosion _returnvalue_6 _returnvalue_3 (ScalableFloat|GetValueatLevel (Variables|Default|GetKnockbackIntensity)) (ScalableFloat|GetValueatLevel (Variables|Default|GetMinKnockbackAngle)))
    ;[1 1 1 0.5] Applying Fall Damage Immunity
    (bind _returnvalue_7 (Spec|MakeSpecHandle(ByClass) (Variables|Default|GetGE_FallDamageImmunity) self))
    (Fortnite|ApplyGameplayEffectSpecToActorsInRange _returnvalue_7 (Variables|Default|GetObject%20Types) _returnvalue_6 _returnvalue_3 (Variables|Default|GetIgnored%20Actors) false (Ability|GetAbilitySystemComponent (Game|GetPlayerPawn 0)))
    (CallFunction|CheckforLaunchEffects _returnvalue_6 _returnvalue_3)))

(event Custom|FuseTimerMax
  (bind _yknockbackoffset (Variables|Default|GetYKnockbackOffset))
  (bind _returnvalue (ScalableFloat|GetValueatLevel _yknockbackoffset))
  (bind _zknockbackoffset (Variables|Default|GetZKnockbackOffset))
  (bind _returnvalue_2 (ScalableFloat|GetValueatLevel _zknockbackoffset))
  (bind _explosionradius (Variables|Default|GetExplosionRadius))
  (bind _returnvalue_3 (ScalableFloat|GetValueatLevel _explosionradius))
  (bind _collisionmesh (Variables|Default|GetCollisionMesh))
  (bind _returnvalue_4 (Transformation|GetWorldTransform _collisionmesh))
  (bind _returnvalue_5 (Math|Transform|TransformDirection _returnvalue_4 (Math|Vector|MakeVector -5.0 _returnvalue _returnvalue_2)))
  (bind _mesh (Variables|Projectile_Rocket|GetMesh))
  (bind _returnvalue_6 (Transformation|GetWorldLocation _mesh))
  (bind _decallocation (Variables|Effects|Explosion|Decal|GetDecalLocation))
  (bind _returnvalue_7 (+ _decallocation _returnvalue_5))
  ;[1 1 1] EVENT -> Stop: Impact and Explosion Stuff
  (Variables|Effects|Explosion|SetTimerMaxReached? true)
  (Variables|Effects|SetStopLocation _returnvalue_6)
  (if (not (Delivery|IsBeingKilled))
    (Delivery|DoExplosion)
    (Fortnite|ShowSoundIndicatorLocal :Location _returnvalue_6 :MaxAudibleDistance 5e+03 :IndicatorType "Gunshot" :Instigator self :Tint "(R=0.000000,G=0.000000,B=0.000000,A=0.000000)" :AttenuationMaxRangeMultiplier 1.0 :AttenuationIntensityMultiplier 1.0)
    ;[0.003247 0.6351 1 0.5] Knockback
    (Class|AthenaFunctionLibrary|CreateEnvironmentalKnockbackExplosion _returnvalue_7 _returnvalue_3 (ScalableFloat|GetValueatLevel (Variables|Default|GetKnockbackIntensity)) (ScalableFloat|GetValueatLevel (Variables|Default|GetMinKnockbackAngle)))
    ;[1 1 1 0.5] Applying Fall Damage Immunity
    (bind _returnvalue_8 (Spec|MakeSpecHandle(ByClass) (Variables|Default|GetGE_FallDamageImmunity) self))
    (Fortnite|ApplyGameplayEffectSpecToActorsInRange _returnvalue_8 (Variables|Default|GetObject%20Types) _returnvalue_7 _returnvalue_3 (Variables|Default|GetIgnored%20Actors) false (Ability|GetAbilitySystemComponent (Game|GetPlayerPawn 0)))
    (CallFunction|CheckforLaunchEffects _returnvalue_7 _returnvalue_3)))

(event EventOnExploded
  (params (HitActors :type Array%20of%20Actor%20Object%20References) (HitResults :type Array%20of%20Hit%20Result%20Structures))
  (bind _rockettrailps_niagara (Variables|Default|GetRocketTrailPS_Niagara))
  (bind _rockettrailps (Variables|RocketTrail|GetRocketTrailPS))
  ;[1 0.5861 0.08121] EVENT -> Exploded
  (Utilities|FlowControl|DoOnce false
    (:Completed
      (Networking|CanExecuteCosmeticEvents
        (:True
          (Development|SetHiddeninGame (Variables|Projectile_Rocket|GetMesh) true true)
          (Utilities|IsValid _rockettrailps
            (:Is%20Valid
              (Components|Activation|Deactivate _rockettrailps))
            (:Is%20Not%20Valid
              _))
          (Utilities|IsValid _rockettrailps_niagara
            (:Is%20Valid
              (Components|Activation|Deactivate _rockettrailps_niagara))
            (:Is%20Not%20Valid
              _))
          (if (or (Utilities|Enum|Equal(Enum) (Variables|Effects|GetSurfaceType) "SurfaceType8") (Water|IsinWater (Variables|Water|GetWaterInteractionComponent)))
            ;;[0.6757 0 1] Spawn Water FX and Sound
            (Class|BPSurfaceTypeSoundComponent|SetSurfaceType (Variables|Sound|GetBP_SurfaceTypeSoundComponent) "SurfaceType8" 0)
            (Effects|Components|ParticleSystem|SpawnFXAtLocation (Variables|Effects|Explosion|GetExplosion%20Asset) (Variables|Effects|GetStopLocation) (Variables|Effects|GetStopRotZ))
            (Effects|Components|ParticleSystem|SpawnFXAtLocation (Variables|Effects|Explosion|GetExplosion%20Asset_Water) (Variables|Effects|GetStopLocation) (Variables|Effects|GetStopRotZ))
            (Utilities|SpawnFNRadialForce (Utilities|StopLocationTransform) (Variables|PhysicsFields|GetPlayrate) (Variables|PhysicsFields|GetRadius) (Variables|PhysicsFields|GetStrength) (Variables|PhysicsFields|GetTime%20Decay))
            (Audio|PlaySoundatLocation (Variables|Sounds|GetRocket_Projectile_Explosion_Water_Sound) (Variables|Effects|GetStopLocation))
            (else
              ;;[0.2043 0 1 1 30] Niagara Explosion Branch
              (Effects|Components|ParticleSystem|SpawnFXAtLocation (Variables|Effects|Explosion|GetExplosion%20Asset) (Variables|Effects|GetStopLocation) (Variables|Effects|GetStopRotZ))
              (Utilities|SpawnFNRadialForce (Utilities|StopLocationTransform) (Variables|PhysicsFields|GetPlayrate) (Variables|PhysicsFields|GetRadius) (Variables|PhysicsFields|GetStrength) (Variables|PhysicsFields|GetTime%20Decay))
              (Audio|PlaySoundatLocation (Variables|Sounds|GetRocket_Projectile_Explosion_Sound) (Variables|Effects|GetStopLocation))))
          (Class|BPSurfaceTypeSoundComponent|PlaySound (Variables|Sound|GetBP_SurfaceTypeSoundComponent) (Variables|Effects|GetStopLocation))
          ;;[0.6661 0.1729 1] Shake that camera / Post Process
          (Utilities|FlowControl|Sequence
            (:then_0
              (if (<= (Math|Vector|VectorLength (- (Variables|Effects|GetStopLocation) (Transformation|GetWorldLocation (Variables|/Script/Engine.PlayerCameraManager|GetTransformComponent (Game|GetPlayerCameraManager 0))))) (Variables|/Script/Engine.PostProcessComponent|GetBlendRadius (Variables|Default|GetPP_Explode)))
                (bind (_direction _ppfade) (Timeline|PPFader PlayFromStart)
                  (:Update
                    (Variables|/Script/Engine.PostProcessComponent|SetBlendWeight _ppfade (Variables|Default|GetPP_Explode))))))
            (:then_1
              (GameplayCue|ExecuteGameplayCue%28Burst%29Local self (Variables|Effects|GetFeedbackCue) (Ability|GameplayCue|MakeGameplayCueParameters 0.0)) ; Cam shake + controller vibration
              (Audio|Components|Audio|FadeOut (Variables|Default|GetRocket_AudioComponent) 1 0.0 "Linear" :!class /Script/Engine.AudioComponent))))))))

(event EventOnDisableForKill
  (bind _rockettrailps_niagara (Variables|Default|GetRocketTrailPS_Niagara))
  (bind _rockettrailps (Variables|RocketTrail|GetRocketTrailPS))
  ;[1 0.5861 0.08121] EVENT -> Exploded
  (Utilities|FlowControl|DoOnce false
    (:Completed
      (Networking|CanExecuteCosmeticEvents
        (:True
          (Development|SetHiddeninGame (Variables|Projectile_Rocket|GetMesh) true true)
          (Utilities|IsValid _rockettrailps
            (:Is%20Valid
              (Components|Activation|Deactivate _rockettrailps))
            (:Is%20Not%20Valid
              _))
          (Utilities|IsValid _rockettrailps_niagara
            (:Is%20Valid
              (Components|Activation|Deactivate _rockettrailps_niagara))
            (:Is%20Not%20Valid
              _))
          (if (or (Utilities|Enum|Equal(Enum) (Variables|Effects|GetSurfaceType) "SurfaceType8") (Water|IsinWater (Variables|Water|GetWaterInteractionComponent)))
            ;;[0.6757 0 1] Spawn Water FX and Sound
            (Class|BPSurfaceTypeSoundComponent|SetSurfaceType (Variables|Sound|GetBP_SurfaceTypeSoundComponent) "SurfaceType8" 0)
            (Effects|Components|ParticleSystem|SpawnFXAtLocation (Variables|Effects|Explosion|GetExplosion%20Asset) (Variables|Effects|GetStopLocation) (Variables|Effects|GetStopRotZ))
            (Effects|Components|ParticleSystem|SpawnFXAtLocation (Variables|Effects|Explosion|GetExplosion%20Asset_Water) (Variables|Effects|GetStopLocation) (Variables|Effects|GetStopRotZ))
            (Utilities|SpawnFNRadialForce (Utilities|StopLocationTransform) (Variables|PhysicsFields|GetPlayrate) (Variables|PhysicsFields|GetRadius) (Variables|PhysicsFields|GetStrength) (Variables|PhysicsFields|GetTime%20Decay))
            (Audio|PlaySoundatLocation (Variables|Sounds|GetRocket_Projectile_Explosion_Water_Sound) (Variables|Effects|GetStopLocation))
            (else
              ;;[0.2043 0 1 1 30] Niagara Explosion Branch
              (Effects|Components|ParticleSystem|SpawnFXAtLocation (Variables|Effects|Explosion|GetExplosion%20Asset) (Variables|Effects|GetStopLocation) (Variables|Effects|GetStopRotZ))
              (Utilities|SpawnFNRadialForce (Utilities|StopLocationTransform) (Variables|PhysicsFields|GetPlayrate) (Variables|PhysicsFields|GetRadius) (Variables|PhysicsFields|GetStrength) (Variables|PhysicsFields|GetTime%20Decay))
              (Audio|PlaySoundatLocation (Variables|Sounds|GetRocket_Projectile_Explosion_Sound) (Variables|Effects|GetStopLocation))))
          (Class|BPSurfaceTypeSoundComponent|PlaySound (Variables|Sound|GetBP_SurfaceTypeSoundComponent) (Variables|Effects|GetStopLocation))
          ;;[0.6661 0.1729 1] Shake that camera / Post Process
          (Utilities|FlowControl|Sequence
            (:then_0
              (if (<= (Math|Vector|VectorLength (- (Variables|Effects|GetStopLocation) (Transformation|GetWorldLocation (Variables|/Script/Engine.PlayerCameraManager|GetTransformComponent (Game|GetPlayerCameraManager 0))))) (Variables|/Script/Engine.PostProcessComponent|GetBlendRadius (Variables|Default|GetPP_Explode)))
                (Timeline|PPFader PlayFromStart)))
            (:then_1
              (GameplayCue|ExecuteGameplayCue%28Burst%29Local self (Variables|Effects|GetFeedbackCue) (Ability|GameplayCue|MakeGameplayCueParameters 0.0)) ; Cam shake + controller vibration
              (Audio|Components|Audio|FadeOut (Variables|Default|GetRocket_AudioComponent) 1 0.0 "Linear" :!class /Script/Engine.AudioComponent))))))))

(event EventDestroyed
  (bind _rockettrailps (Variables|RocketTrail|GetRocketTrailPS))
  ;[1 1 1] EVENT -> Destroyed: Deactivating Rocket Trail Particle systems on component destroyed
  (Networking|CanExecuteCosmeticEvents
    (:True
      (Utilities|IsValid _rockettrailps
        (:Is%20Valid
          _)
        (:Is%20Not%20Valid
          (Utilities|IsValid (Variables|Default|GetRocketTrailPS_Niagara)
            (:Is%20Valid
              _))))
      (Components|Activation|Deactivate _rockettrailps))))

(fn CheckForLaunchEffects
  (params (Start :type Vector) (Radius :type Float%20%28single-precision%29))
  (bind _outhits (Collision|MultiSphereTraceForObjects Start Start Radius (Variables|Default|GetObject%20Types) false (Variables|Default|GetIgnored%20Actors)))
  (for _array_element _outhits
    (Utilities|IsValid (.HitActor (Collision|BreakHitResult _array_element))
      (:Is%20Valid
        (GameplayCue|ExecuteGameplayCueOnActor(Burst) (.HitActor (Collision|BreakHitResult _array_element)) (Variables|Sound|GetLaunch_GameplayCue) (Ability|GameplayCue|MakeGameplayCueParameters :NormalizedMagnitude 0.0 :RawMagnitude 0.0 :Location "0,0,0" :Normal "0,0,0" :Instigator self :EffectCauser self :SourceObject (.HitActor (Collision|BreakHitResult _array_element)) :GameplayEffectLevel 1 :AbilityLevel 1 :bReplicateLocationWhenUsingMinimalRepProxy false))))))

(fn ConstructionScript
  (Utilities|FlowControl|Sequence
    (:then_0
      (Effects|Components|ParticleSystem|SetTemplate (Variables|RocketTrail|GetRibbon_Trail_PSC) (Variables|RocketTrail|GetRibbon_Trail_PS))
      (Effects|Components|ParticleSystem|SetTemplate (Variables|Default|GetExplosion_Flesh_Damage_PSC) (Variables|Effects|Explosion|Surfaces|GetExplosion_Flesh_Damage_PS))
      (Transformation|SetRelativeScale3D (Variables|Default|GetCollisionMesh) (Variables|Projectile_Rocket|GetRocketScale)))
    (:then_1
      ;[1 1 1] Load Rocket Trail asset
      (Variables|RocketTrail|SetRocket%20Trial%20Asset (select (Effects|IsNiagaraFXTypeEnabled "Projectile") (Variables|Default|GetNewNiagaraRocketTrailAsset) "/Game/Weapons/FORT_RocketLaunchers/Effects/NS_RocketTrail_01_Athena.NS_RocketTrail_01_Athena")))
    (:then_2
      ;[1 1 1] Load Explosion asset
      (Variables|Effects|Explosion|SetExplosion%20Asset (Variables|Default|GetNewNiagaraExplosionAsset)))))
