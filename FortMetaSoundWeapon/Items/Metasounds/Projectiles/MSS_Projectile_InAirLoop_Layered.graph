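# Layered stereo MetaSound source for looping projectile in-air sound, mixing random close and distant waves with shared random pitch.
# Distance curves set each layer's gain, with stereo close starting at a random offset and distant mono.
# Output gain combines level, an affiliation crossfade (local, friendly, enemy) and a fade-in envelope.
# Mobile page plays both layers mono without fade or offset; random volume is stored but never applied on either page.
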MetaSoundSource /FortMetaSoundWeapon/Items/Metasounds/Projectiles/MSS_Projectile_InAirLoop_Layered
inputs [
 Affiliation.Enemy {type=Float}
 Affiliation.Friendly {type=Float}
 Affiliation.Local {type=Float}
 Character.Affiliation {type=Enum:FortAudioAffiliation}
 Character.Health {type=Float}
 Character.UniqueID {type=Int32}
 Projectile.CloseVolume {type=Float}
 Projectile.DistanceCurves {type=WaveTableBankAsset}
 Projectile.DistantVolume {type=Float}
 Projectile.FadeInTime {type=Time}
 Projectile.FalloffDistance {type=Float}
 Projectile.InnerRadius {type=Float}
 Projectile.OutputVolume {type=Float}
 "Random.Max Pitch (Semitones)" {type=Float}
 Random.MinPitch {type=Float}
 Random.VolumeMax {type=Float}
 Random.VolumeMin {type=Float}
 Sounds.Close {type=WaveAsset:Array}
 Sounds.Distant {type=WaveAsset:Array}
 UE.Attenuation.Distance {type=Float}
 UE.Source.OnPlay {type=Trigger}
]
interfaces [Fort.Character:1.0, UE.Attenuation:1.0, UE.OutputFormat.Stereo:1.0, UE.Source:1.1]
outputs [
 "UE.OutputFormat.Stereo.Audio:0" {type=Audio}
 "UE.OutputFormat.Stereo.Audio:1" {type=Audio}
]
pages [
 Default {
  is_build_graph=true
  input_defaults {
   Affiliation.Enemy=0
   Affiliation.Friendly=-6
   Affiliation.Local=0
   Character.Affiliation=0
   Character.Health=0
   Character.UniqueID=0
   Projectile.CloseVolume=0
   Projectile.DistanceCurves=/FortMetaSoundWeapon/Items/Shared/Settings/Projectiles/WTB_Projectile_InAir_Distance
   Projectile.DistantVolume=0
   Projectile.FadeInTime=0.3
   Projectile.FalloffDistance=27500
   Projectile.InnerRadius=0
   Projectile.OutputVolume=0
   "Random.Max Pitch (Semitones)"=0
   Random.MinPitch=0
   Random.VolumeMax=0
   Random.VolumeMin=0
   UE.Attenuation.Distance=14519.9922
   UE.Source.OnPlay=null
   Sounds.Close []
   Sounds.Distant []
  }
  nodes [
   DecibelstoLinearGain_6 = UE.Decibels to Linear Gain:Float(Decibels=^Projectile.CloseVolume)
   n1 = UE.Multiply:Float(
    AdditionalOperands=@DecibelstoLinearGain_6.Linear Gain
    PrimaryOperand=&MSP_Attenuation_Volume_Curve(
     AttenuationCurves=^Projectile.DistanceCurves
     FalloffDistance=^Projectile.FalloffDistance
     InnerRadius=^Projectile.InnerRadius
     Distance=^UE.Attenuation.Distance
    ).VolumeMultiplier
   )
   n2 = UE.RandomFloat(Max=^"Random.Max Pitch (Semitones)", Min=^Random.MinPitch, Next=^UE.Source.OnPlay)
   n3 = UE.RandomFloat(Max=^Random.VolumeMax, Min=^Random.VolumeMin, Next=@n2.On Next)
   ArrayRandomGet_1 = Array.Random Get:WaveAsset:Array(In Array=^Sounds.Close, Next=@n3.On Next) {also {defaults {No Repeats=2}}}
   GetWaveDuration_1 = "UE.Get Wave Duration: "(Wave=@ArrayRandomGet_1.Value)
   n4 = UE.RandomTime(Next=@ArrayRandomGet_1.On Next, Max=@GetWaveDuration_1.Duration)
   n5 = VariableAccessor.Float(Variable=VariableMutator.Float(Variable=InitVariable.Float().Variable, Value=@n2.Value).Variable)
   WavePlayer_1 = UE.Wave Player:Stereo(Wave Asset=@ArrayRandomGet_1.Value, Play=@n4.On Next, Start Time=@n4.Value, Pitch Shift=@n5.Value) {also {defaults {Loop="<MetaSoundBuilderResult.SUCCEEDED: 0>"}}}
   Multiply_0 = UE.Multiply:Audio by Float(AdditionalOperands=@n1.Out, PrimaryOperand=@WavePlayer_1.Out Right)
   DecibelstoLinearGain_1 = UE.Decibels to Linear Gain:Float(Decibels=^Projectile.DistantVolume)
   n6 = &MSP_Attenuation_Volume_Curve(
    AttenuationCurves=^Projectile.DistanceCurves
    FalloffDistance=^Projectile.FalloffDistance
    InnerRadius=^Projectile.InnerRadius
    Distance=^UE.Attenuation.Distance
   ) {also {defaults {CurveIndexToEvaluate=1}}}
   ArrayRandomGet_0 = Array.Random Get:WaveAsset:Array(In Array=^Sounds.Distant, Next=@n3.On Next) {also {defaults {No Repeats=2, Weights []}}}
   GetWaveDuration_0 = "UE.Get Wave Duration: "(Wave=@ArrayRandomGet_0.Value)
   WavePlayer_0 = UE.Wave Player:Mono(
    Wave Asset=@ArrayRandomGet_0.Value
    Play=UE.RandomTime(Next=@ArrayRandomGet_0.On Next, Max=@GetWaveDuration_0.Duration).On Next
    Pitch Shift=VariableAccessor.Float(Variable=@n5.Variable).Value
   ) {also {defaults {Loop="<MetaSoundBuilderResult.SUCCEEDED: 0>"}}}
   n7 = UE.Multiply:Audio by Float(PrimaryOperand=UE.Multiply:Audio by Float(
     AdditionalOperands=UE.Multiply:Float(PrimaryOperand=@DecibelstoLinearGain_1.Linear Gain, AdditionalOperands=@n6.VolumeMultiplier).Out
     PrimaryOperand=@WavePlayer_0.Out Mono
    ).Out) {also {defaults {AdditionalOperands=0.7}}}
   DecibelstoLinearGain_2 = UE.Decibels to Linear Gain:Float(Decibels=^Projectile.OutputVolume)
   ADSREnvelopeADSREnvelope_0 = ADSR Envelope.ADSR Envelope:Float(Attack Time=^Projectile.FadeInTime, Trigger Attack=^UE.Source.OnPlay) {also {defaults {Sustain Level=1}}}
   DecibelstoLinearGain_0 = UE.Decibels to Linear Gain:Float(Decibels=^Affiliation.Local)
   DecibelstoLinearGain_3 = UE.Decibels to Linear Gain:Float(Decibels=^Affiliation.Friendly)
   DecibelstoLinearGain_4 = UE.Decibels to Linear Gain:Float(Decibels=^Affiliation.Enemy)
   CrossfadeTriggerRouteFloat4_0 = "Crossfade.Trigger Route (Float, 4):Float"(
    Crossfade Value=Convert.Float:Int32(Int32=Convert.Int32:Enum:FortAudioAffiliation("Enum:FortAudioAffiliation"=^Character.Affiliation).Int32).Float
    In 0=@DecibelstoLinearGain_0.Linear Gain
    In 1=@DecibelstoLinearGain_3.Linear Gain
    In 2=@DecibelstoLinearGain_4.Linear Gain
    In 3=@DecibelstoLinearGain_4.Linear Gain
   )
   n8 = UE.Multiply:Float(
    PrimaryOperand=@DecibelstoLinearGain_2.Linear Gain
    AdditionalOperands=UE.Multiply:Float(AdditionalOperands=@ADSREnvelopeADSREnvelope_0.Out Envelope, PrimaryOperand=@CrossfadeTriggerRouteFloat4_0.Out).Out
   )
   Multiply_3 = UE.Multiply:Audio by Float(PrimaryOperand=UE.Add:Audio(PrimaryOperand=@Multiply_0.Out, AdditionalOperands=@n7.Out).Out, AdditionalOperands=@n8.Out)
   Multiply_4 = UE.Multiply:Audio by Float(
    PrimaryOperand=UE.Add:Audio(PrimaryOperand=UE.Multiply:Audio by Float(AdditionalOperands=@n1.Out, PrimaryOperand=@WavePlayer_1.Out Left).Out, AdditionalOperands=@n7.Out).Out
    AdditionalOperands=@n8.Out
   )
   DecibelstoLinearGain_5 = UE.Decibels to Linear Gain:Float(Decibels=@n3.Value)
   VariableMutator.Float(Value=@DecibelstoLinearGain_5.Linear Gain, Variable=InitVariable.Float().Variable)
  ]
  links [{dst="@UE.OutputFormat.Stereo.Audio:1", src=Multiply_3.Out}, {dst="@UE.OutputFormat.Stereo.Audio:0", src=Multiply_4.Out}]
 }
 Mobile {
  nodes [
   DecibelstoLinearGain_6 = UE.Decibels to Linear Gain:Float(Decibels=^Projectile.CloseVolume)
   Multiply_4 = UE.Multiply:Float(
    AdditionalOperands=@DecibelstoLinearGain_6.Linear Gain
    PrimaryOperand=&MSP_Attenuation_Volume_Curve(
     AttenuationCurves=^Projectile.DistanceCurves
     FalloffDistance=^Projectile.FalloffDistance
     InnerRadius=^Projectile.InnerRadius
     Distance=^UE.Attenuation.Distance
    ).VolumeMultiplier
   )
   n1 = UE.RandomFloat(Max=^"Random.Max Pitch (Semitones)", Min=^Random.MinPitch, Next=^UE.Source.OnPlay)
   n2 = UE.RandomFloat(Max=^Random.VolumeMax, Min=^Random.VolumeMin, Next=@n1.On Next)
   ArrayRandomGet_1 = Array.Random Get:WaveAsset:Array(In Array=^Sounds.Close, Next=@n2.On Next) {also {defaults {No Repeats=2}}}
   n3 = VariableAccessor.Float(Variable=VariableMutator.Float(Variable=InitVariable.Float().Variable, Value=@n1.Value).Variable)
   WavePlayer_1 = UE.Wave Player:Mono(Play=@ArrayRandomGet_1.On Next, Wave Asset=@ArrayRandomGet_1.Value, Pitch Shift=@n3.Value) {also {defaults {Loop="<MetaSoundBuilderResult.SUCCEEDED: 0>"}}}
   DecibelstoLinearGain_1 = UE.Decibels to Linear Gain:Float(Decibels=^Projectile.DistantVolume)
   n4 = UE.Multiply:Float(
    PrimaryOperand=@DecibelstoLinearGain_1.Linear Gain
    AdditionalOperands=&MSP_Attenuation_Volume_Curve(
     AttenuationCurves=^Projectile.DistanceCurves
     FalloffDistance=^Projectile.FalloffDistance
     InnerRadius=^Projectile.InnerRadius
     Distance=^UE.Attenuation.Distance
    ) {also {defaults {CurveIndexToEvaluate=1}}}.VolumeMultiplier
   )
   ArrayRandomGet_0 = Array.Random Get:WaveAsset:Array(In Array=^Sounds.Distant, Next=@n2.On Next) {also {defaults {No Repeats=2, Weights []}}}
   WavePlayer_0 = UE.Wave Player:Mono(Play=@ArrayRandomGet_0.On Next, Wave Asset=@ArrayRandomGet_0.Value, Pitch Shift=VariableAccessor.Float(Variable=@n3.Variable).Value) {also {defaults {Loop="<MetaSoundBuilderResult.SUCCEEDED: 0>"}}}
   DecibelstoLinearGain_0 = UE.Decibels to Linear Gain:Float(Decibels=^Affiliation.Local)
   DecibelstoLinearGain_3 = UE.Decibels to Linear Gain:Float(Decibels=^Affiliation.Friendly)
   DecibelstoLinearGain_4 = UE.Decibels to Linear Gain:Float(Decibels=^Affiliation.Enemy)
   CrossfadeTriggerRouteFloat4_0 = "Crossfade.Trigger Route (Float, 4):Float"(
    Crossfade Value=Convert.Float:Int32(Int32=Convert.Int32:Enum:FortAudioAffiliation("Enum:FortAudioAffiliation"=^Character.Affiliation).Int32).Float
    In 0=@DecibelstoLinearGain_0.Linear Gain
    In 1=@DecibelstoLinearGain_3.Linear Gain
    In 2=@DecibelstoLinearGain_4.Linear Gain
    In 3=@DecibelstoLinearGain_4.Linear Gain
   )
   DecibelstoLinearGain_2 = UE.Decibels to Linear Gain:Float(Decibels=^Projectile.OutputVolume)
   Multiply_3 = UE.Multiply:Audio by Float(
    PrimaryOperand=UE.Add:Audio(
     PrimaryOperand=UE.Multiply:Audio by Float(AdditionalOperands=@Multiply_4.Out, PrimaryOperand=@WavePlayer_1.Out Mono).Out
     AdditionalOperands=UE.Multiply:Audio by Float(AdditionalOperands=@n4.Out, PrimaryOperand=@WavePlayer_0.Out Mono).Out
    ).Out
    AdditionalOperands=UE.Multiply:Float(AdditionalOperands=@CrossfadeTriggerRouteFloat4_0.Out, PrimaryOperand=@DecibelstoLinearGain_2.Linear Gain).Out
   )
   Multiply_0 = UE.Multiply:Audio by Float(PrimaryOperand=@Multiply_3.Out) {also {defaults {AdditionalOperands=0.7}}}
   DecibelstoLinearGain_5 = UE.Decibels to Linear Gain:Float(Decibels=@n2.Value)
   VariableMutator.Float(Value=@DecibelstoLinearGain_5.Linear Gain, Variable=InitVariable.Float().Variable)
  ]
  links [{dst="@UE.OutputFormat.Stereo.Audio:0", src=Multiply_0.Out}, {dst="@UE.OutputFormat.Stereo.Audio:1", src=Multiply_0.Out}]
 }
]
properties {
 assetImportData=&self:AssetImportData
 attenuationSettings=/FortMetaSoundWeapon/Items/Shared/Settings/Projectiles/ATT_Projectile_InAir_NoAtt
 bLooping=true
 bOverrideConcurrency=true
 duration=10000
 maxDistance=27500
 numChannels=2
 outputFormat=Stereo
 sampleRate=48000
 soundClassObject=&Fort_WeaponRanged_Projectile_Effects
 busSends [{audioBus=&AB_Sidechain_FortWeapon_Vol, sendLevel=1}]
 concurrencyOverrides {maxCount=3}
 concurrencySet [null]
}
