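# Looping stereo MetaSound source for the in-flight sound of VividRazor rocket launcher projectiles, a preset of the shared layered in-air projectile loop.
# It supplies PumpkinLauncher close and distant loops with layer volumes, a fade-in and shared wavetable distance curves, using a NoAtt attenuation asset.
# Friendly affiliation is set quieter, and random pitch and volume are zeroed.
# It overrides concurrency with a low voice cap and sends to a weapon sidechain bus.
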MetaSoundSource /VividRazorExotics/Gameplay/RocketLauncher/Sounds/MSS_VividRazor_RocketLauncher_Projectile_Loop
preset_parent=/FortMetaSoundWeapon/Items/Metasounds/Projectiles/MSS_Projectile_InAirLoop_Layered
inputs [
 Affiliation.Enemy {type=Float}
 Affiliation.Friendly {type=Float}
 Affiliation.Local {type=Float}
 Character.Affiliation {type=Enum:FortAudioAffiliation}
 Character.Health {type=Float}
 Character.UniqueID {type=Int32}
 Projectile.CloseVolume {type=Float}
 Projectile.DistanceCurves {type=WaveTableBankAsset}
 Projectile.DistantVolume {type=Float}
 Projectile.FadeInTime {type=Time}
 Projectile.FalloffDistance {type=Float}
 Projectile.InnerRadius {type=Float}
 Projectile.OutputVolume {type=Float}
 "Random.Max Pitch (Semitones)" {type=Float}
 Random.MinPitch {type=Float}
 Random.VolumeMax {type=Float}
 Random.VolumeMin {type=Float}
 Sounds.Close {type=WaveAsset:Array}
 Sounds.Distant {type=WaveAsset:Array}
 UE.Attenuation.Distance {type=Float}
 UE.Source.OnPlay {type=Trigger}
]
interfaces [Fort.Character:1.0, UE.Attenuation:1.0, UE.OutputFormat.Stereo:1.0, UE.Source:1.1]
outputs [
 "UE.OutputFormat.Stereo.Audio:0" {type=Audio}
 "UE.OutputFormat.Stereo.Audio:1" {type=Audio}
]
pages [Default {input_defaults {
   Affiliation.Enemy=0
   Affiliation.Friendly=-6
   Affiliation.Local=0
   Character.Affiliation=0
   Character.Health=0
   Character.UniqueID=0
   Projectile.CloseVolume=0
   Projectile.DistanceCurves=/FortMetaSoundWeapon/Items/Shared/Settings/Projectiles/WTB_Projectile_InAir_Distance
   Projectile.DistantVolume=-9
   Projectile.FadeInTime=0.3
   Projectile.FalloffDistance=27500
   Projectile.InnerRadius=0
   Projectile.OutputVolume=-3
   "Random.Max Pitch (Semitones)"=0
   Random.MinPitch=0
   Random.VolumeMax=0
   Random.VolumeMin=0
   UE.Attenuation.Distance=14519.9922
   UE.Source.OnPlay=null
   Sounds.Close [/Game/Sounds/Fort_Weapons/PumpkinLauncher/SoundWaves/PumpkinLauncher_Prj_InAir_Close_Loop]
   Sounds.Distant [/Game/Sounds/Fort_Weapons/PumpkinLauncher/SoundWaves/PumpkinLauncher_Prj_InAir_Loop]
  }}]
properties {
 assetImportData=&self:AssetImportData
 attenuationSettings=/FortMetaSoundWeapon/Items/Shared/Settings/Projectiles/ATT_Projectile_InAir_NoAtt
 bLooping=true
 bOverrideConcurrency=true
 duration=10000
 maxDistance=27500
 numChannels=2
 outputFormat=Stereo
 sampleRate=48000
 soundClassObject=&Fort_WeaponRanged_Projectile_Effects
 busSends [{audioBus=&AB_Sidechain_FortWeapon_Vol, sendLevel=1}]
 concurrencyOverrides {maxCount=3}
 concurrencySet [null]
}
