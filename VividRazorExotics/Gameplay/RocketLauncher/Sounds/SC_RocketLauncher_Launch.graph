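# Sound cue for the rocket launcher launch sound in VividRazorExotics, reusing Fortnite low-gravity-stone jump recordings.
# A Random node picks evenly between two jump waves, then a Modulator slightly raises pitch and roughly doubles volume with small variation.
# It uses a dedicated launch attenuation and the gadgets-items sound class.
# A concurrency override allows one instance, blocking new plays while active and throttling rapid retriggers.
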SoundCue /VividRazorExotics/Gameplay/RocketLauncher/Sounds/SC_RocketLauncher_Launch
root=Modulator_1
nodes [
 {id=Modulator_1, children [Random_0], properties {pitchMax=1.1, pitchMin=1, volumeMax=2.2, volumeMin=2}}
 {id=Random_0, children [WavePlayer_0,WavePlayer_2], properties {weights [1,1]}}
 {id=WavePlayer_0, properties {soundWaveAssetPtr=/Game/Sounds/Fort_GamePlay_Sounds/Player/LowGravityStones/LowGravityStone_Jump_01}}
 {id=WavePlayer_2, properties {soundWaveAssetPtr=/FNGameplayCues/Sounds/Fort_GamePlay_Sounds/Player/LowGravityStones/LowGravityStone_Jump_03}}
]
properties {
 attenuationSettings=&/ATT_RocketLauncher_Launch
 bOverrideConcurrency=true
 maxDistance=10000
 soundClassObject=&Fort_Gadgets_Items
 volumeMultiplier=1
 concurrencyOverrides {maxCount=1, resolutionRule=PreventNew, retriggerTime=0.3}
}
