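# Looping in-air sound cue for the RocketLauncher projectile, reusing PumpkinLauncher in-air wave loops.
# Two layers, a default loop and a louder-mixed close loop, each run through slight random pitch, a volume fade-in envelope, strong Doppler and their own attenuation asset.
# A mixer blends both layers under a ranged-weapon third-person sound class, with a small concurrency cap limiting simultaneous instances.
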SoundCue /VividRazorExotics/Gameplay/RocketLauncher/Sounds/SC_RocketLauncher_Prj_InAir_Loop
root=Mixer_1
nodes [
 {id=Attenuation_0, children [Doppler_3], properties {attenuationSettings=/Game/Sounds/Attenuation/Fort_Weapon_Launcher_Proj_InAir_Close_Att}}
 {
  id=Attenuation_3
  children [Doppler_0]
  properties {attenuationSettings=/Game/Sounds/Attenuation/Fort_Weapon_Launcher_Proj_InAir_Default_Att}
 }
 {id=Doppler_0, children [Enveloper_1], properties {dopplerIntensity=4}}
 {id=Doppler_3, children [Enveloper_0], properties {dopplerIntensity=4}}
 {
  id=Enveloper_0
  children [Modulator_6]
  properties {pitchCurve {editorCurveData {keys [{time=0}]}}, volumeCurve {editorCurveData {keys [<at 0 0>,<at 1 1>]}}}
 }
 {
  id=Enveloper_1
  children [Modulator_1]
  properties {pitchCurve {editorCurveData {keys [{time=0}]}}, volumeCurve {editorCurveData {keys [<at 0 0>,<at 1 1>]}}}
 }
 {id=Mixer_1, children [Attenuation_3, Attenuation_0], properties {inputVolume [1,2]}}
 {id=Modulator_1, children [WavePlayer_0], properties {pitchMax=1.1, pitchMin=0.9, volumeMax=1, volumeMin=1}}
 {id=Modulator_6, children [WavePlayer_1], properties {pitchMax=1.1, pitchMin=0.9, volumeMax=1, volumeMin=1}}
 {
  id=WavePlayer_0
  properties {bLooping=true, soundWaveAssetPtr=/Game/Sounds/Fort_Weapons/PumpkinLauncher/SoundWaves/PumpkinLauncher_Prj_InAir_Loop}
 }
 {
  id=WavePlayer_1
  properties {bLooping=true, soundWaveAssetPtr=/Game/Sounds/Fort_Weapons/PumpkinLauncher/SoundWaves/PumpkinLauncher_Prj_InAir_Close_Loop}
 }
]
properties {
 bDebug=true
 bOverrideConcurrency=true
 maxDistance=12750
 soundClassObject=&Fort_WeaponRanged_Fire_3P_PSM
 volumeMultiplier=0.7
 attenuationOverrides {falloffDistance=3400, attenuationShapeExtents {x=600}}
 concurrencyOverrides {maxCount=3}
}
